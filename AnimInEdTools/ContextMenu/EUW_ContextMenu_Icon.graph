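# EUW_ContextMenu_Icon is an Editor Utility Widget that lays out one context menu row with an icon and a label.
# A HorizontalBox holds a size-capped Icon image, hidden by default, beside a filling ButtonText label.
# It declares a String Data variable and a Context Menu Item Clicked dispatcher carrying text.
# Its event graph has no logic, so nothing here fills the row or fires the dispatcher.
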WidgetBlueprint /AnimInEdTools/ContextMenu/EUW_ContextMenu_Icon
graphs [
 Context Menu Item Clicked {nodes [Context Menu Item Clicked!() -> (then, Text:String) {is=title K2Node_FunctionEntry}]}
 EventGraph {nodes [
   Event PreConstruct()
   Event Construct()
   Event Tick()
  ]}
]
variables [
 Context Menu Item Clicked {}
 String Data {type=string, default=""}
]
widgets {
 info {inheritedWidgetCount=0, parentClass=&EditorUtilityWidget, rootWidgetClass=&HorizontalBox, widgetCount=4}
 widgets [
  HorizontalBox_1 {widgetClassPath=&HorizontalBox, widget=HorizontalBox_1}
  SizeBox_0 {
   widgetClassPath=&SizeBox
   parent=HorizontalBox_1
   widget=SizeBox_0
   properties {bOverride_MaxDesiredHeight=true, bOverride_MaxDesiredWidth=true, maxDesiredHeight=40, maxDesiredWidth=40}
   slot {class=HorizontalBoxSlot, properties {verticalAlignment=VAlign_Center, padding=<ltrb 12 0 0 0>}}
  }
  Icon {
   widgetClassPath=&Image
   bIsVariable=true
   parent=SizeBox_0
   widget=Icon
   properties {
    visibility=Hidden
    brush {
     bIsDynamicallyLoaded=false
     drawAs=Image
     imageType=NoImage
     mirroring=NoMirror
     resourceName=None
     resourceObject=null
     tiling=NoTile
     imageSize=<16 16>
     margin=<ltrb 0 0 0 0>
     outlineSettings {
      bUseBrushTransparency=false
      roundingType=HalfHeightRadius
      width=0
      color {colorUseRule=UseColor_Specified, specifiedColor=<rgba 0 0 0 0>}
      cornerRadii=<0 0 0 0>
     }
     tintColor {colorUseRule=UseColor_Specified, specifiedColor=<rgba 1 1 1 1>}
     uVRegion {bIsValid=false, max=<0 0>, min=<0 0>}
    }
   }
   slot {class=SizeBoxSlot}
  }
  ButtonText {
   widgetClassPath=&TextBlock
   bIsVariable=true
   parent=HorizontalBox_1
   widget=ButtonText
   properties {
    text=Animation Type
    font {
     bForceMonospaced=false
     bMaterialIsStencil=false
     fontMaterial=null
     fontObject=&Roboto
     letterSpacing=0
     monospacedWidth=1
     size=10
     skewAmount=0
     typefaceFontName=Regular
     outlineSettings {
      bApplyOutlineToDropShadows=false
      bMiteredCorners=false
      bSeparateFillAlpha=false
      outlineMaterial=null
      outlineSize=0
      outlineColor=<rgba 0 0 0 1>
     }
    }
   }
   slot {class=HorizontalBoxSlot, properties {verticalAlignment=VAlign_Center, padding=<ltrb 12 1 0 1>, size {sizeRule=Fill, value=1}}}
  }
 ]
}
