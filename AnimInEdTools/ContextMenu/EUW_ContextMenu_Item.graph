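# Editor utility widget for a single clickable item in a context menu, styled as a flat button with a blue hover highlight.
# A named slot fills the button to hold the item's content.
# On click it broadcasts Context Menu Item Clicked with itself and its Slot Widget reference, then hides the owning Context Menu if valid.
# The Close Menu On Click variable is declared but not read in this graph.
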WidgetBlueprint /AnimInEdTools/ContextMenu/EUW_ContextMenu_Item
graphs [
 Context Menu Item Clicked {nodes [Context Menu Item Clicked!() -> (then, Context Item:Editor Utility Widget Object Reference, Slot Widget:Editor Utility Widget Object Reference) {is=title K2Node_FunctionEntry}]}
 EventGraph {nodes [
   "On Clicked (Button)"() {is=title K2Node_ComponentBoundEvent}
   > Call Context Menu Item Clicked!(
    execute:Exec
    self:EUW Context Menu Item Object Reference
    Context Item:Editor Utility Widget Object Reference=Self-Reference()
    Slot Widget:Editor Utility Widget Object Reference=Get Slot Widget() -> (Slot Widget:Editor Utility Widget Object Reference)
   ) -> (then:Exec) {is=title K2Node_CallDelegate}
   n1 = Get Context Menu() -> (Context Menu:EUW Context Menu Object Reference)
   > Is Valid(InputObject=@n1)
   > Show!(execute, self:EUW Context Menu Object Reference=@n1, Show:bool=false, Mouse Position:Vector 2D Structure, Name:String) -> (then)
   Event PreConstruct()
   Event Construct()
   Event Tick()
  ]}
]
variables [
 Close Menu On Click {type=boolean}
 Context Menu {type=EUW_ContextMenu_C, default=null}
 Context Menu Item Clicked {}
 Slot Widget {type=EditorUtilityWidget, default=null}
]
widgets {
 info {inheritedWidgetCount=0, parentClass=&EditorUtilityWidget, rootWidgetClass=&Button, widgetCount=3}
 widgets [
  Button {
   widgetClassPath=&Button
   widget=Button
   properties {widgetStyle {
     clickedSlateSound {resourceObject=null}
     disabled {
      bIsDynamicallyLoaded=false
      drawAs=NoDrawType
      imageType=NoImage
      mirroring=NoMirror
      resourceName=None
      resourceObject=null
      tiling=NoTile
      imageSize=<0 0>
      margin=<ltrb 0 0 0 0>
      outlineSettings {
       bUseBrushTransparency=false
       roundingType=FixedRadius
       width=1
       color {colorUseRule=UseColor_ColorTable, specifiedColor=<rgba 0 0 0 0>}
       cornerRadii=<4 4 4 4>
      }
      tintColor {colorUseRule=UseColor_Specified, specifiedColor=<rgba 0.0395 0.0395 0.0395 1>}
      uVRegion {bIsValid=false, max=<0 0>, min=<0 0>}
     }
     disabledForeground {colorUseRule=UseColor_ColorTable, specifiedColor=<rgba 0 0 0 0>}
     hovered {
      bIsDynamicallyLoaded=false
      drawAs=Box
      imageType=NoImage
      mirroring=NoMirror
      resourceName=None
      resourceObject=null
      tiling=NoTile
      imageSize=<32 32>
      margin=<ltrb 0 0 0 0>
      outlineSettings {
       bUseBrushTransparency=false
       roundingType=FixedRadius
       width=1
       color {colorUseRule=UseColor_ColorTable, specifiedColor=<rgba 0 0 0 0>}
       cornerRadii=<4 4 4 4>
      }
      tintColor {colorUseRule=UseColor_Specified, specifiedColor=<rgba 0 0.162 0.7454 1>}
      uVRegion {bIsValid=false, max=<0 0>, min=<0 0>}
     }
     hoveredForeground {colorUseRule=UseColor_ColorTable, specifiedColor=<rgba 0 0 0 0>}
     hoveredSlateSound {resourceObject=null}
     normal {
      bIsDynamicallyLoaded=false
      drawAs=Box
      imageType=NoImage
      mirroring=NoMirror
      resourceName=None
      resourceObject=null
      tiling=NoTile
      imageSize=<32 32>
      margin=<ltrb 0 0 0 0>
      outlineSettings {
       bUseBrushTransparency=false
       roundingType=FixedRadius
       width=1
       color {colorUseRule=UseColor_ColorTable, specifiedColor=<rgba 0 0 0 0>}
       cornerRadii=<4 4 4 4>
      }
      tintColor {colorUseRule=UseColor_Specified, specifiedColor=<rgba 0.0395 0.0395 0.0395 0>}
      uVRegion {bIsValid=false, max=<0 0>, min=<0 0>}
     }
     normalForeground {colorUseRule=UseColor_ColorTable, specifiedColor=<rgba 0 0 0 0>}
     normalPadding=<ltrb 0 0 0 0>
     pressed {
      bIsDynamicallyLoaded=false
      drawAs=NoDrawType
      imageType=NoImage
      mirroring=NoMirror
      resourceName=None
      resourceObject=null
      tiling=NoTile
      imageSize=<32 32>
      margin=<ltrb 0 0 0 0>
      outlineSettings {
       bUseBrushTransparency=false
       roundingType=FixedRadius
       width=1
       color {colorUseRule=UseColor_ColorTable, specifiedColor=<rgba 0 0 0 0>}
       cornerRadii=<4 4 4 4>
      }
      tintColor {colorUseRule=UseColor_Specified, specifiedColor=<rgba 0.0284 0.0284 0.0284 1>}
      uVRegion {bIsValid=false, max=<0 0>, min=<0 0>}
     }
     pressedForeground {colorUseRule=UseColor_ColorTable, specifiedColor=<rgba 0 0 0 0>}
     pressedPadding=<ltrb 0 0 0 0>
     pressedSlateSound {resourceObject=null}
    }}
  }
  HorizontalBox_0 {
   widgetClassPath=&HorizontalBox
   parent=Button
   widget=HorizontalBox_0
   slot {class=ButtonSlot, properties {horizontalAlignment=HAlign_Fill, verticalAlignment=VAlign_Fill, padding=<ltrb 0 0 0 0>}}
  }
  NamedSlot {
   widgetClassPath=&NamedSlot
   bIsVariable=true
   parent=HorizontalBox_0
   widget=NamedSlot
   slot {class=HorizontalBoxSlot, properties {size {sizeRule=Fill, value=1}}}
  }
 ]
}
