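# Editor Utility Widget for a scrollable pop-up context menu in AnimInEdTools.
# Add Item appends a click-bound entry hosting any slot widget, preceded by an uppercase header when its category is new; Add Icon Item adds text and icon.
# Show toggles visibility and render-translates the menu to the mouse, flipping it at parent edges.
# Tick repeats positioning for about 0.2 seconds after showing, using updated cached geometry and clamping horizontally.
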EditorUtilityWidgetBlueprint /AnimInEdTools/ContextMenu/EUW_ContextMenu
graphs [
 Add Icon Item {nodes [
   n1 = Add Icon Item!() -> (then, Event:"Delegate (by ref)", Category:String, Name:String, Font Size:double=10, Icon:Texture 2D Object Reference) {is=title K2Node_FunctionEntry}
   > n2 = Sequence()
   > n3 = Add Item!(
    execute
    self
    Event:"Delegate (by ref)"=@n1
    Slot Widget:User Widget Class Reference=&/EUW_ContextMenu_Icon.EUW_ContextMenu_Icon_C
    Category:String=@n1.Category
   ) -> (then, Context Menu Item:EUW Context Menu Item Object Reference, Widget:Editor Utility Widget Object Reference)
   > n4 = Cast To EUW_ContextMenu_Icon!(execute, Object=@n3.Widget) -> (then, CastFailed, AsEUW Context Menu Icon:EUW Context Menu Icon Object Reference) {is=title K2Node_DynamicCast}
   n5 = Get ButtonText!(self:EUW Context Menu Icon Object Reference=@n4) -> (ButtonText:Text Object Reference)
   > Set Text(Text="To Text (String)"(InString=@n1.Name), self=@n5)
   > Set Font!(
    execute
    Font:Slate Font Info Structure=MakeSlateFontInfo!(
     FontObject:Object Reference=&Roboto
     FontMaterial:Object Reference
     OutlineSettings:Font Outline Settings Structure=(OutlineSize=0,bMiteredCorners=False,bSeparateFillAlpha=False,bApplyOutlineToDropShadows=False,OutlineMaterial=None,OutlineColor=(R=0,G=0,B=0,A=1))
     TypefaceFontName:Name=Regular
     Size:float=@n1.Font Size
     LetterSpacing:int=0
     SkewAmount:float=0
     bForceMonospaced:bool=false
     bMaterialIsStencil:bool=false
     MonospacedWidth:float=1
    ) -> (ReturnValue:Slate Font Info Structure)
    self:Text Object Reference=@n5
   ) -> (then, Output_Get:Slate Font Info Structure)
   > Is Valid(InputObject=@n1.Icon)
   n6 = Get Icon!(self:EUW Context Menu Icon Object Reference=@n4) -> (Icon:Image Object Reference)
   > SetBrush(
    self=@n6
    InBrush=MakeBrushFromTexture!(Texture:Texture 2D Object Reference=@n1.Icon, Width:int=0, Height:int=0) -> (ReturnValue:Slate Brush Structure)
   )
   > SetVisibility(self=@n6, InVisibility=Visible)
   Return Node!(execute=@n2.then_1, Context Menu Icon:EUW Context Menu Icon Object Reference=@n4)
  ]}
 Add Item {nodes [
   n1 = Add Item!() -> (then, Event:"Delegate (by ref)", Slot Widget:User Widget Class Reference, Category:String) {is=title K2Node_FunctionEntry}
   n2 = Get Created Categoriess() -> (Created Categoriess:Array of Strings)
   > n3 = Branch(Condition=OR Boolean(A=IsEmpty(InString=@n1.Category), B=Contains Item(TargetArray:Array of Strings=@n2, ItemToFind:"String (by ref)"=@n1.Category)))
   n4 = Create Widget!(
    execute:Exec=[@n3.then, @n8.then]
    WorldContextObject:Object Reference
    Class:User Widget Class Reference=&/EUW_ContextMenu_Item.EUW_ContextMenu_Item_C
    OwningPlayer:Player Controller Object Reference
   ) -> (then:Exec, ReturnValue:EUW Context Menu Item Object Reference)
   > AddChild(self=Get VerticalBox() -> (VerticalBox:Vertical Box Object Reference), Content=@n4)
   > Set Context Menu!(execute, Context Menu:EUW Context Menu Object Reference=Self-Reference(), self:EUW Context Menu Item Object Reference=@n4) -> (then, Output_Get:EUW Context Menu Object Reference)
   > Bind Event to Context Menu Item Clicked(self:EUW Context Menu Item Object Reference=@n4, Delegate=Get Event() -> (Event:"Delegate (by ref)")) {is=title K2Node_AddDelegate}
   > n5 = Create Widget(Class=Get Slot Widget() -> (Slot Widget:User Widget Class Reference)) -> (ReturnValue:User Widget Object Reference)
   > AddChild(self=Get NamedSlot!(self:EUW Context Menu Item Object Reference=@n4) -> (NamedSlot:Named Slot Object Reference), Content=@n5)
   > n6 = Cast To EditorUtilityWidget!(execute, Object=@n5) -> (then, CastFailed, AsEditor Utility Widget:Editor Utility Widget Object Reference) {is=title K2Node_DynamicCast}
   > Set Slot Widget!(execute, Slot Widget:Editor Utility Widget Object Reference=@n6, self:EUW Context Menu Item Object Reference=@n4) -> (then, Output_Get:Editor Utility Widget Object Reference)
   > Return Node!(execute, Context Menu Item:EUW Context Menu Item Object Reference=@n4, Widget:Editor Utility Widget Object Reference=@n6)
   Add(execute=@n3.else, TargetArray=@n2, NewItem=@n1.Category) {is=title K2Node_CallArrayFunction}
   > n7 = Create Widget(Class=&/EUW_ContextMenu_CategoryItem.EUW_ContextMenu_CategoryItem_C) -> (ReturnValue:EUW Context Menu Category Item Object Reference)
   > AddChild(self=Get VerticalBox(), Content=@n7)
   > n8 = Set Text(
    Text="To Text (String)"(InString=ToUpper!(SourceString:String=@n1.Category) -> (ReturnValue:String))
    self=Get ButtonText(self:EUW Context Menu Category Item Object Reference=@n7)
   )
  ]}
 Clear Context Menu {nodes [
   Clear Context Menu() {is=title K2Node_FunctionEntry}
   > ClearChildren(self=Get VerticalBox())
   > Clear(TargetArray=Get Created Categoriess())
   > Set Name(Name:String) -> (Output_Get:String)
  ]}
 EventGraph {nodes [
   Event Tick()
   > n1 = Branch(Condition=Get Need to Position() -> (Need to Position:bool))
   > Position Menu!(execute, self, Mouse Position:Vector 2D Structure=Get Target Position() -> (Target Position:Vector 2D Structure)) -> (then)
   > Do Once(Reset=@n1.else)
   > Delay!(execute, WorldContextObject:Object Reference, Duration=0.2) -> (then)
   > Set Need to Position(Need to Position:bool=false) -> (Output_Get:bool)
   Event PreConstruct()
   Event Construct()
  ]}
 Position Menu {nodes [
   n1 = Position Menu!() -> (then, Mouse Position:Vector 2D Structure) {is=title K2Node_FunctionEntry}
   > Sequence()
   n2 = GetCachedGeometry!(self:Widget Object Reference=GetParent!(self:Widget Object Reference) -> (ReturnValue:Panel Widget Object Reference)) -> (ReturnValue:"Geometry Structure (by ref)")
   n3 = BreakVector2D(InVec="Absolute to Local (Coordinates)"!(Geometry:"Geometry Structure (by ref)"=@n2, AbsoluteCoordinate:Vector 2D Structure=@n1) -> (ReturnValue:Vector 2D Structure))
   n4 = GetLocalSize!(Geometry:"Geometry Structure (by ref)"=@n2) -> (ReturnValue_X:double, ReturnValue_Y:double)
   n5 = GetLocalSize(Geometry=GetCachedGeometry(self=Self-Reference()))
   n6 = "float > float"(A=float + float(A=@n5.ReturnValue_Y, B=@n3.Y), B=@n4.ReturnValue_Y)
   n7 = float - float(A=@n3.Y, B=@n5.ReturnValue_Y)
   n8 = float + float(A=@n5, B=@n3)
   n9 = "Max (Float)"(A=SelectFloat(A=float - float(A=@n3, B=@n5), B=@n3, bPickA="float > float"(A=@n8, B=@n4)), B=0)
   n10 = "float <= float"(A=@n4, B=@n5) -> (ReturnValue=~)
   > SetRenderTransform!(
    execute
    self:Widget Object Reference
    InTransform:Widget Transform Structure=Make Widget Transform!(
     Translation_X:double=SelectFloat(A=0, B=@n9, bPickA=@n10)
     Translation_Y:double=SelectFloat(A=@n7, B=@n3.Y, bPickA=@n6)
     Scale:Vector 2D Structure=(X=1,Y=1)
     Shear:Vector 2D Structure=(X=0,Y=0)
     Angle:float=0
    ) -> (WidgetTransform:Widget Transform Structure) {is=title K2Node_MakeStruct}
   ) -> (then)
  ]}
 Show {nodes [
   n1 = Sequence()
   n2 = GetCachedGeometry(self=GetParent())
   n3 = Show!() -> (then, Show:bool=~, Mouse Position:Vector 2D Structure, Name:String) {is=title K2Node_FunctionEntry}
   n4 = BreakVector2D(InVec="Absolute to Local (Coordinates)"(Geometry=@n2, AbsoluteCoordinate=@n3.Mouse Position))
   n5 = GetLocalSize(Geometry=@n2)
   n6 = GetLocalSize(Geometry=GetCachedGeometry(self=Self-Reference()))
   n7 = "float > float"(A=float + float(A=@n6.ReturnValue_Y, B=@n4.Y), B=@n5.ReturnValue_Y)
   n8 = float - float(A=@n4.Y, B=@n6.ReturnValue_Y)
   n9 = "float > float"(A=float + float(A=@n6, B=@n4), B=@n5)
   n10 = float - float(A=@n4, B=@n6)
   SetRenderTransform(
    execute=@n1.then_0
    InTransform=Make Widget Transform(
     Translation_X=SelectFloat(A=@n10, B=@n4, bPickA=@n9)
     Translation_Y=SelectFloat(A=@n8, B=@n4.Y, bPickA=@n7)
     Scale=(X=1,Y=1)
     Shear=(X=0,Y=0)
     Angle=0
    )
   )
   SetVisibility(execute=@n11.then, InVisibility=Select(Option 0=Collapsed, Option 1=Visible, Index=@n3))
   Set Need to Position(execute=@n12.then, Need to Position=@n3)
   > n11 = Set Name(Name=@n3.Name)
   n12 = Set Target Position(execute=@n3.then, Target Position:Vector 2D Structure=@n3.Mouse Position) -> (Output_Get:Vector 2D Structure)
  ]}
]
variables [
 Created Categoriess {type="array<string>", default []}
 Name {type=string, default=""}
 Need to Position {type=boolean}
 Target Position {type=Vector2D, default=<0 0>}
]
widgets {
 info {rootWidgetClass=&Border, widgetCount=4}
 widgets [
  Border {
   widgetClassPath=&Border
   bIsVariable=true
   widget=Border
   properties {brushColor=<rgba 0.0395 0.0395 0.0395 1>, padding=<ltrb 8 8 8 8>}
  }
  SizeBox {
   widgetClassPath=&SizeBox
   bIsVariable=true
   parent=Border
   widget=SizeBox
   slot {class=BorderSlot, properties {padding=<ltrb 8 8 8 8>}}
  }
  ScrollBox {
   widgetClassPath=&ScrollBox
   bIsVariable=true
   parent=SizeBox
   widget=ScrollBox
   properties {
    widgetBarStyle {
     thickness=8
     draggedThumbImage {
      bIsDynamicallyLoaded=false
      drawAs=RoundedBox
      imageType=NoImage
      mirroring=NoMirror
      resourceName=None
      resourceObject=null
      tiling=NoTile
      imageSize=<0 0>
      margin=<ltrb 0 0 0 0>
      outlineSettings {
       bUseBrushTransparency=false
       roundingType=FixedRadius
       width=0
       color {colorUseRule=UseColor_Specified, specifiedColor=<rgba 0 0 0 0>}
       cornerRadii=<4 4 4 4>
      }
      tintColor {colorUseRule=UseColor_Specified, specifiedColor=<rgba 0.2159 0.2159 0.2159 1>}
      uVRegion {bIsValid=false, max=<0 0>, min=<0 0>}
     }
     horizontalBackgroundImage {
      bIsDynamicallyLoaded=false
      drawAs=NoDrawType
      imageType=NoImage
      mirroring=NoMirror
      resourceName=None
      resourceObject=null
      tiling=NoTile
      imageSize=<0 0>
      margin=<ltrb 0 0 0 0>
      outlineSettings {
       bUseBrushTransparency=false
       roundingType=HalfHeightRadius
       width=0
       color {colorUseRule=UseColor_Specified, specifiedColor=<rgba 0 0 0 0>}
       cornerRadii=<0 0 0 0>
      }
      tintColor {colorUseRule=UseColor_Specified, specifiedColor=<rgba 1 1 1 1>}
      uVRegion {bIsValid=false, max=<0 0>, min=<0 0>}
     }
     horizontalBottomSlotImage {
      bIsDynamicallyLoaded=false
      drawAs=NoDrawType
      imageType=NoImage
      mirroring=NoMirror
      resourceName=None
      resourceObject=null
      tiling=NoTile
      imageSize=<0 0>
      margin=<ltrb 0 0 0 0>
      outlineSettings {
       bUseBrushTransparency=false
       roundingType=HalfHeightRadius
       width=0
       color {colorUseRule=UseColor_Specified, specifiedColor=<rgba 0 0 0 0>}
       cornerRadii=<0 0 0 0>
      }
      tintColor {colorUseRule=UseColor_Specified, specifiedColor=<rgba 1 1 1 1>}
      uVRegion {bIsValid=false, max=<0 0>, min=<0 0>}
     }
     horizontalTopSlotImage {
      bIsDynamicallyLoaded=false
      drawAs=NoDrawType
      imageType=NoImage
      mirroring=NoMirror
      resourceName=None
      resourceObject=null
      tiling=NoTile
      imageSize=<0 0>
      margin=<ltrb 0 0 0 0>
      outlineSettings {
       bUseBrushTransparency=false
       roundingType=HalfHeightRadius
       width=0
       color {colorUseRule=UseColor_Specified, specifiedColor=<rgba 0 0 0 0>}
       cornerRadii=<0 0 0 0>
      }
      tintColor {colorUseRule=UseColor_Specified, specifiedColor=<rgba 1 1 1 1>}
      uVRegion {bIsValid=false, max=<0 0>, min=<0 0>}
     }
     hoveredThumbImage {
      bIsDynamicallyLoaded=false
      drawAs=RoundedBox
      imageType=NoImage
      mirroring=NoMirror
      resourceName=None
      resourceObject=null
      tiling=NoTile
      imageSize=<0 0>
      margin=<ltrb 0 0 0 0>
      outlineSettings {
       bUseBrushTransparency=false
       roundingType=FixedRadius
       width=0
       color {colorUseRule=UseColor_Specified, specifiedColor=<rgba 0 0 0 0>}
       cornerRadii=<4 4 4 4>
      }
      tintColor {colorUseRule=UseColor_Specified, specifiedColor=<rgba 0.2159 0.2159 0.2159 1>}
      uVRegion {bIsValid=false, max=<0 0>, min=<0 0>}
     }
     normalThumbImage {
      bIsDynamicallyLoaded=false
      drawAs=RoundedBox
      imageType=NoImage
      mirroring=NoMirror
      resourceName=None
      resourceObject=null
      tiling=NoTile
      imageSize=<0 0>
      margin=<ltrb 0 0 0 0>
      outlineSettings {
       bUseBrushTransparency=false
       roundingType=FixedRadius
       width=0
       color {colorUseRule=UseColor_Specified, specifiedColor=<rgba 0 0 0 0>}
       cornerRadii=<4 4 4 4>
      }
      tintColor {colorUseRule=UseColor_Specified, specifiedColor=<rgba 0.0953 0.0953 0.0953 1>}
      uVRegion {bIsValid=false, max=<0 0>, min=<0 0>}
     }
     verticalBackgroundImage {
      bIsDynamicallyLoaded=false
      drawAs=NoDrawType
      imageType=NoImage
      mirroring=NoMirror
      resourceName=None
      resourceObject=null
      tiling=NoTile
      imageSize=<0 0>
      margin=<ltrb 0 0 0 0>
      outlineSettings {
       bUseBrushTransparency=false
       roundingType=HalfHeightRadius
       width=0
       color {colorUseRule=UseColor_Specified, specifiedColor=<rgba 0 0 0 0>}
       cornerRadii=<0 0 0 0>
      }
      tintColor {colorUseRule=UseColor_Specified, specifiedColor=<rgba 1 1 1 1>}
      uVRegion {bIsValid=false, max=<0 0>, min=<0 0>}
     }
     verticalBottomSlotImage {
      bIsDynamicallyLoaded=false
      drawAs=NoDrawType
      imageType=NoImage
      mirroring=NoMirror
      resourceName=None
      resourceObject=null
      tiling=NoTile
      imageSize=<0 0>
      margin=<ltrb 0 0 0 0>
      outlineSettings {
       bUseBrushTransparency=false
       roundingType=HalfHeightRadius
       width=0
       color {colorUseRule=UseColor_Specified, specifiedColor=<rgba 0 0 0 0>}
       cornerRadii=<0 0 0 0>
      }
      tintColor {colorUseRule=UseColor_Specified, specifiedColor=<rgba 1 1 1 1>}
      uVRegion {bIsValid=false, max=<0 0>, min=<0 0>}
     }
     verticalTopSlotImage {
      bIsDynamicallyLoaded=false
      drawAs=NoDrawType
      imageType=NoImage
      mirroring=NoMirror
      resourceName=None
      resourceObject=null
      tiling=NoTile
      imageSize=<0 0>
      margin=<ltrb 0 0 0 0>
      outlineSettings {
       bUseBrushTransparency=false
       roundingType=HalfHeightRadius
       width=0
       color {colorUseRule=UseColor_Specified, specifiedColor=<rgba 0 0 0 0>}
       cornerRadii=<0 0 0 0>
      }
      tintColor {colorUseRule=UseColor_Specified, specifiedColor=<rgba 1 1 1 1>}
      uVRegion {bIsValid=false, max=<0 0>, min=<0 0>}
     }
    }
    widgetStyle {
     barThickness=8
     bottomShadowBrush {
      bIsDynamicallyLoaded=false
      drawAs=Box
      imageType=FullColor
      mirroring=NoMirror
      resourceName=../../../Engine/Content/Slate/Common/ScrollBoxShadowBottom.png
      resourceObject=null
      tiling=NoTile
      imageSize=<16 8>
      margin=<ltrb 0.5 0 0.5 1>
      outlineSettings {
       bUseBrushTransparency=false
       roundingType=HalfHeightRadius
       width=0
       color {colorUseRule=UseColor_Specified, specifiedColor=<rgba 0 0 0 0>}
       cornerRadii=<0 0 0 0>
      }
      tintColor {colorUseRule=UseColor_Specified, specifiedColor=<rgba 1 1 1 1>}
      uVRegion {bIsValid=false, max=<0 0>, min=<0 0>}
     }
     horizontalScrolledContentPadding=<ltrb 0 0 1 0>
     leftShadowBrush {
      bIsDynamicallyLoaded=false
      drawAs=Box
      imageType=FullColor
      mirroring=NoMirror
      resourceName=../../../Engine/Content/Slate/Common/ScrollBoxShadowLeft.png
      resourceObject=null
      tiling=NoTile
      imageSize=<8 16>
      margin=<ltrb 1 0.5 0 0.5>
      outlineSettings {
       bUseBrushTransparency=false
       roundingType=HalfHeightRadius
       width=0
       color {colorUseRule=UseColor_Specified, specifiedColor=<rgba 0 0 0 0>}
       cornerRadii=<0 0 0 0>
      }
      tintColor {colorUseRule=UseColor_Specified, specifiedColor=<rgba 1 1 1 1>}
      uVRegion {bIsValid=false, max=<0 0>, min=<0 0>}
     }
     rightShadowBrush {
      bIsDynamicallyLoaded=false
      drawAs=Box
      imageType=FullColor
      mirroring=NoMirror
      resourceName=../../../Engine/Content/Slate/Common/ScrollBoxShadowRight.png
      resourceObject=null
      tiling=NoTile
      imageSize=<8 16>
      margin=<ltrb 0 0.5 1 0.5>
      outlineSettings {
       bUseBrushTransparency=false
       roundingType=HalfHeightRadius
       width=0
       color {colorUseRule=UseColor_Specified, specifiedColor=<rgba 0 0 0 0>}
       cornerRadii=<0 0 0 0>
      }
      tintColor {colorUseRule=UseColor_Specified, specifiedColor=<rgba 1 1 1 1>}
      uVRegion {bIsValid=false, max=<0 0>, min=<0 0>}
     }
     topShadowBrush {
      bIsDynamicallyLoaded=false
      drawAs=Box
      imageType=FullColor
      mirroring=NoMirror
      resourceName=../../../Engine/Content/Slate/Common/ScrollBoxShadowTop.png
      resourceObject=null
      tiling=NoTile
      imageSize=<16 8>
      margin=<ltrb 0.5 1 0.5 0>
      outlineSettings {
       bUseBrushTransparency=false
       roundingType=HalfHeightRadius
       width=0
       color {colorUseRule=UseColor_Specified, specifiedColor=<rgba 0 0 0 0>}
       cornerRadii=<0 0 0 0>
      }
      tintColor {colorUseRule=UseColor_Specified, specifiedColor=<rgba 1 1 1 1>}
      uVRegion {bIsValid=false, max=<0 0>, min=<0 0>}
     }
     verticalScrolledContentPadding=<ltrb 0 0 0 1>
    }
   }
   slot {class=SizeBoxSlot}
  }
  VerticalBox {
   widgetClassPath=&VerticalBox
   bIsVariable=true
   parent=ScrollBox
   widget=VerticalBox
   slot {class=ScrollBoxSlot, properties {size {sizeRule=Fill, value=1}}}
  }
 ]
}
