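# Editor utility widget for a checkable context menu item: a non-interactive checkbox beside a text label in a minimum-sized box.
# SetIsChecked sets the checkbox to checked or unchecked, then refreshes the label from the String Data variable on both branches.
# A Context Menu Item Clicked dispatcher carries a Text string, but nothing in this widget broadcasts it, and the event graph holds only empty events.
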WidgetBlueprint /AnimInEdTools/ContextMenu/EUW_ContextMenu_ItemB
graphs [
 Context Menu Item Clicked {nodes [Context Menu Item Clicked!() -> (then, Text:String) {is=title K2Node_FunctionEntry}]}
 EventGraph {nodes [
   Event PreConstruct()
   Event Construct()
   Event Tick()
  ]}
 SetIsChecked {nodes [
   n1 = SetIsChecked!() -> (then, bIsChecked:bool=~) {is=title K2Node_FunctionEntry}
   > n2 = Branch(Condition=@n1)
   n3 = Get CheckBox_IsChecked() -> (CheckBox_IsChecked:Check Box Object Reference)
   > n4 = Set CheckedState!(execute, CheckedState:ECheckBoxState Enum=Checked, self:Check Box Object Reference=@n3) -> (then, Output_Get:ECheckBoxState Enum=Unchecked)
   "SetText (Text)"(
    execute=[@n4.then, @n5.then]
    self=Get ButtonText() -> (ButtonText:Text Object Reference)
    InText="To Text (String)"(InString=Get String Data() -> (String Data:String))
   )
   n5 = Set CheckedState(execute=@n2.else, CheckedState=Unchecked, self=@n3) -> (Output_Get=Unchecked)
  ]}
]
variables [
 Context Menu Item Clicked {}
 String Data {type=string, default=""}
]
widgets {
 info {inheritedWidgetCount=0, parentClass=&EditorUtilityWidget, rootWidgetClass=&HorizontalBox, widgetCount=5}
 widgets [
  HorizontalBox_1 {widgetClassPath=&HorizontalBox, widget=HorizontalBox_1}
  SizeBox_0 {
   widgetClassPath=&SizeBox
   parent=HorizontalBox_1
   widget=SizeBox_0
   properties {bOverride_MinDesiredHeight=true, bOverride_MinDesiredWidth=true, minDesiredHeight=20, minDesiredWidth=90}
   slot {class=HorizontalBoxSlot}
  }
  HorizontalBox_0 {widgetClassPath=&HorizontalBox, parent=SizeBox_0, widget=HorizontalBox_0, slot {class=SizeBoxSlot}}
  CheckBox_IsChecked {
   widgetClassPath=/Script/UMG.CheckBox
   bIsVariable=true
   parent=HorizontalBox_0
   widget=CheckBox_IsChecked
   properties {visibility=HitTestInvisible}
   slot {class=HorizontalBoxSlot}
  }
  ButtonText {
   widgetClassPath=&TextBlock
   bIsVariable=true
   parent=HorizontalBox_0
   widget=ButtonText
   properties {
    text=Bookmark1
    font {
     bForceMonospaced=false
     bMaterialIsStencil=false
     fontMaterial=null
     fontObject=&Roboto
     letterSpacing=0
     monospacedWidth=1
     size=10
     skewAmount=0
     typefaceFontName=Regular
     outlineSettings {
      bApplyOutlineToDropShadows=false
      bMiteredCorners=false
      bSeparateFillAlpha=false
      outlineMaterial=null
      outlineSize=0
      outlineColor=<rgba 0 0 0 1>
     }
    }
   }
   slot {class=HorizontalBoxSlot, properties {verticalAlignment=VAlign_Center, padding=<ltrb 12 0 12 0>}}
  }
 ]
}
